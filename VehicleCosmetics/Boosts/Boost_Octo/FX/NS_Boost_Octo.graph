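# Niagara system for the Octo vehicle boost cosmetic, drawing exhaust sprites and trailing ribbons at the car's boost sockets.
# A burst of looping flipbook sprites alternates between two socket indices and visibility tags across two renderers.
# Per-port ribbons stream backward, with layered sine offsets on particle position creating a wavy motion that widens with age.
# Painted-variant overrides and the bIsPainted flag switch color, brightness and dynamic material parameters.
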NiagaraSystem /VehicleCosmetics/Boosts/Boost_Octo/FX/NS_Boost_Octo
dependencies {
 usedDataInterfaces [
  &NiagaraDataInterfaceCurve
  &NiagaraDataInterfaceSpriteRendererInfo
  &NiagaraDataInterfaceArrayFloat
  &NiagaraDataInterfaceVectorCurve
  &NiagaraDataInterfaceSkeletalMesh
 ]
 usedDynamicInputs []
 usedModules [
  &self:SystemSpawnScript.SystemScriptSource.SystemScriptGraph.NiagaraNodeAssignment_0.SetVariables_#1
  &SystemState
 ]
}
emitters {
 NE_VC_Boost_Ribbon_Parent004 {
  emitterSpawnScript [/SetVariables(
    Emitter.ColorBool:NiagaraFloat=&MakeCustomFloatFromBool(Bool=$User.bIsPainted, True Float=0, False Float=1)
    Emitter.Contrast:NiagaraFloat=&SelectFloatFromArray(
     Direct Array Index=$System.ColorVariantID
     Float Selection Array=DataInterface() {propertyValues {FloatData [0,0,1,1,0.9,1.1,1,1,1.1,1.2,1,1,1,1]}}
    )
    Emitter.DefaulTColor:NiagaraBool=&Boolean_NotOperation(Bool=$User.bIsPainted)
   ) {isSetParameters=true}]
  emitterUpdateScript [
   &VC_RateQualityHandler(High Spawn Rate=60, Low Spawn Rate=60, Medium Spawn Rate=60, Use Custom Rate?=true) {module=VC_RateQualityHandler001}
   &EmitterState()
   &VC_FilteredSocketCount()
   &VC_RateQualityHandler(Default Rate=45) {enabled=false}
   &VC_SpawnRibbon_v1(Number of Ribbons=$Emitter.PortCount, Spawn Time=-1, SpawnRate=25)
  ]
  particleSpawnScript [
   &VC_PaintedVariantsOverride(
    Black=<rgba 0.026 0.026 0.026 1>
    BurntSienna=<rgba 0.0642 0.0198 0 1>
    Cobalt=<rgba 0 0.019 0.18 1>
    Crimson=<rgba 0.245 0.015 0.02 1>
    ForestGreen=<rgba 0.005 0.1725 0.005 1>
    Grey=<rgba 0.1 0.1 0.1 1>
    Lime=<rgba 0.12 0.25 0.0025 1>
    Orange=<rgba 0.2957 0.0412 0.0051 1>
    Pink=<rgba 0.1719 0.0349 0.1567 1>
    Purple=<rgba 0.1047 0 0.2552 1>
    Saffron=<rgba 0.2665 0.276 0 1>
    SkyBlue=<rgba 0.0326 0.0851 0.2604 1>
    TitaniumWhite=<rgba 0.2656 0.2656 0.2656 1>
   )
   &VC_InitializeRibbon_v1(
    Color=$StackContext.Color
    Custom Offset=true
    Enable Skeletal Position=true
    Local Vector Offset=&Add_Vector(A=<80 0 0>, B=$User.DTCarOffset)
    Offset Amount=1
    Ribbon Width=55
   )
   &AddVelocity(Velocity=<-1000 0 0>, Velocity Speed Scale=3.5) {module=AddVelocity001}
   &Color(Color=&MakeCustomLinearColorFromBool(Bool=$User.bIsPainted, True Color=$User.PaintedColor2, False Color=<rgba 1.1 1.1 1.1 1>)) {enabled=false}
  ]
  particleUpdateScript [
   &ParticleState()
   &ScaleColor(
    Scale RGB=&Multiply_VectorByFloat(
     Vector=<1 1 1>
     Float=&MakeCustomFloatFromBool(Bool=$User.bIsPainted, True Float=$User.Brightness, False Float=2)
    )
    ScaleA=true
    ScaleRGB=true
   )
   &InheritSourceMovement(
    Applied Position Scale=&VectorFromCurve(VectorCurve=DataInterface() {propertyValues {XCurve {keys [<at 0 1 RCIM_Cubic>,<at 1 0 RCIM_Cubic>]}, YCurve {keys [<at 0 1>,<at 1 1>]}, ZCurve {keys [<at 0 1>,<at 1 1>]}}})
    ApplyForce=false
    ApplyPosition=true
    ApplyPositionFalloff=false
   )
   &SolveForcesAndVelocity(Clamp Velocity=false, Limit Acceleration=false)
   &VC_UpdateRibbon_v1(Enable Skeletal Position=true, Interpolate Bone Sampling=true)
   &DynamicMaterialParameters(
    Index 0 Param 1=&MakeCustomFloatFromBool(Bool=$User.bIsPainted, True Float=0, False Float=1)
    Index 0 Param 2=1
    Index 0 Param 3=1.3
    Module.Index0Param0Write=true
    Module.Index0Param1Write=true
    Module.Index0Param2Write=true
    Module.Index0Param3Write=false
   )
   &ScaleColor(ScaleA=true, ScaleRGB=true) {enabled=false, module=ScaleColor001}
   /SetVariables(Particles.Position=&AddVectorToPosition(
     Position=$Particles.Position
     Vector=&TransformVector(
      Vector=&MakeVector(
       X=&Sine(Normalized Angle=$Emitter.Age, Period=1.5, Scale=4.3, Bias=0)
       Y=&Sine(
        Normalized Angle=&Add_Float(
         A=&Multiply_Float(A=$Particles.NormalizedAge, B=-0.5)
         B=&Multiply_Float(A=$Emitter.LoopedAge, B=1.35)
        )
        Period=0.45
        Scale=&FloatFromCurve(FloatCurve=DataInterface() {propertyValues {Curve {keys [<at 0.1 0>,<at 0.5 1>]}}}, Scale Curve=8)
        Bias=0
       )
       Z=0
      )
      Destination Space=::Simulation
      Apply Scale=true
     )
    )) {isSetParameters=true}
  ]
  rendererClasses [&NiagaraRibbonRendererProperties]
  renderers [
   {rendererClass=&NiagaraRibbonRendererProperties}
   {rendererClass=&NiagaraRibbonRendererProperties, rendererIndex=1}
  ]
 }
 NE_VC_Boost_Sprite_Parent002 {
  emitterUpdateScript [
   &EmitterState(
    Inactive Response=::NewEnumerator1
    Life Cycle Mode=::NewEnumerator1
    Loop Behavior=::NewEnumerator1
    Loop Duration Mode=::NewEnumerator1
    UseLoopDelay=false
   )
   &SpawnBurst_Instantaneous(Spawn Count=16, Use Spawn Probability=false, UseLoopCountLimit=false)
   &VC_FilteredSocketCount()
   &SpawnRate(SpawnRate=2, Use Spawn Probability=false) {enabled=false}
   /SetVariables(Emitter.SingleBoost:NiagaraBool=&SetBoolByFloatComparison(A=&MakeFloatFromInt(Int=$User.PortCount), B=1)) {isSetParameters=true}
  ]
  particleSpawnScript [
   /SetVariables(Particles.CustomIndex:NiagaraInt32=&ModuloInt(A=&ReturnExecIndex(), B=2)) {isSetParameters=true}
   /SetVariables(Particles.VisTag:NiagaraInt32=$Particles.CustomIndex) {isSetParameters=true}
   &VC_PaintedVariantsOverride(
    BurntSienna=<rgba 0.0642 0.0198 0 1.5>
    Cobalt=<rgba 0.0425 0.085 0.5 1>
    Crimson=<rgba 0.245 0.015 0.02 1>
    DefaultColor=<rgba 0.1875 1 0.1875 1>
    Orange=<rgba 0.474 0.1665 0.029 1>
    Pink=<rgba 0.9823 0.2763 0.9495 1>
    Saffron=<rgba 0.5826 0.5544 0 1>
    SkyBlue=<rgba 0.1074 0.3157 0.9823 1>
   )
   &InitializeParticle(
    Color=$StackContext.Color
    Lifetime=&RandomRangeFloat(Minimum=0.8, Maximum=0.5)
    Sprite Rotation Mode=::NewEnumerator2
    Sprite Size Mode=::NewEnumerator3
    Uniform Sprite Size=&Multiply_Float(A=&RandomRangeFloat(Minimum=20, Maximum=14), B=3.5)
    UsePositionOffset=false
    Write Material Random=true
   )
   &PartitionParticles(Number Of Partitions=$Emitter.PortCount)
   &SkeletalMeshLocation(
    Bone / Socket Index=$Particles.CustomIndex
    Bone / Socket Sampling Mode=::NewEnumerator5
    MaskSpawnGroup=false
    Mesh Sampling Type=::NewEnumerator6
    Sampled Position Offset=&AddVectorToPosition(
     Position=&MakeCustomVectorFromBool(Bool=$User.bIsFrontendPreview, True Vector=<-80 0 0>, False Vector=<-80 0 0>)
     Vector=$User.DTCarOffset
    )
    Skeletal Mesh=$System.SkeletalMesh
   )
   &AddVelocity(Velocity=<-1500 0 0>) {enabled=false}
   &Color(Color=&MakeCustomLinearColorFromBool(Bool=$User.bIsPainted, True Color=$User.PaintedColor, False Color=<rgba 0.1875 1 0.1875 1>)) {enabled=false}
   &ShapeLocation(MaskSpawnGroup=false, Override Seed=false, Sphere Radius=10, UseOwnerScale=true)
  ]
  particleUpdateScript [
   &ParticleState(Kill Particles When Lifetime Has Elapsed=false, Loop Particles Lifetime=true)
   &ScaleSpriteSize(Uniform Curve Sprite Scale=DataInterface() {propertyValues {Curve {keys [<at 0 1.5 RCIM_Cubic>,<at 1 1.7 RCIM_Cubic>]}}})
   &ScaleColor(
    Scale RGB=&VectorFromFloat(Value=&FloatFromCurve(FloatCurve=DataInterface() {propertyValues {Curve {keys [<at 0.4 1>,<at 1 0.6>]}}}))
    ScaleA=true
    ScaleRGB=true
   )
   &SolveForcesAndVelocity(Clamp Velocity=false, Limit Acceleration=false)
   &SubUVAnimation(
    "Sprite Renderer "=DataInterface() {propertyValues {SpriteRenderer=&self:NE_VC_Boost_Sprite_Parent002_1.NiagaraSpriteRendererProperties_1}}
    UseEndFrame=false
    UseStartFrame=false
   )
   &SubUVAnimation(
    Output Alternate Renderer Binding=true
    "Sprite Renderer "=DataInterface() {propertyValues {SpriteRenderer=&self:NE_VC_Boost_Sprite_Parent002_1.NiagaraSpriteRendererProperties_2}}
    UseEndFrame=false
    UseStartFrame=false
   ) {module=SubUVAnimation001}
   &DynamicMaterialParameters(
    Index 0 Param 1=&MakeCustomFloatFromBool(
     Bool=$User.bIsPainted
     True Float=1
     False Float=&FloatFromCurve(FloatCurve=DataInterface() {propertyValues {Curve {keys [<at 0 3 RCIM_Cubic>,<at 1 1.5 RCIM_Cubic>]}}})
    )
    Index 0 Param 2=&MakeCustomFloatFromBool(Bool=$User.bIsPainted, True Float=0, False Float=1)
    Module.Index0Param0Write=true
    Module.Index0Param1Write=true
    Module.Index0Param2Write=false
    Module.Index0Param3Write=false
   )
  ]
  rendererClasses [&NiagaraSpriteRendererProperties]
  renderers [
   {rendererClass=&NiagaraSpriteRendererProperties}
   {rendererClass=&NiagaraSpriteRendererProperties, rendererIndex=1}
  ]
 }
}
summary {
 systemName=NS_Boost_Octo
 emitters [
  {emitterName=NE_VC_Boost_Sprite_Parent002, rendererClasses [&NiagaraSpriteRendererProperties]}
  {emitterName=NE_VC_Boost_Ribbon_Parent004, rendererClasses [&NiagaraRibbonRendererProperties]}
 ]
 userVariables [
  User.bIsFrontendPreview {type=NiagaraBool}
  User.bIsPainted {type=NiagaraBool}
  User.Brightness {type=NiagaraFloat, defaultValue=1, description=""}
  User.ColorVariantID {type=NiagaraInt32}
  User.DTCarOffset {type=Vector3f, description="", defaultValue {struct=Vector3f, value=<0 0 0>}}
  User.PaintedColor {type=LinearColor, description="", defaultValue {struct=LinearColor, value=<rgba 1 1 1 1>}}
  User.PaintedColor2 {type=LinearColor, description="", defaultValue {struct=LinearColor, value=<rgba 1 1 1 1>}}
  User.PaintedScalar {type=NiagaraFloat, defaultValue=1, description=""}
  User.PortCount {type=NiagaraInt32, defaultValue=2}
 ]
}
